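# Unlit material for the heart FX mesh in the main player's Swatch emote, supporting Niagara mesh particles.
# Normalized pre-skinned XZ position becomes radial coordinates, with time-panned variants for animated texture effects.
# Texture values offset coordinates feeding a sine-banded brightness blend, tinted by particle color and added to a tinted Fresnel rim.
# The grayscale texture uses mesh-UV derivatives for mip selection instead of its computed UVs.
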Material /Game/Animation/Game/MainPlayer/Emotes/Swatch/FX/M_FX_Emote_Swatch_Heart
graph {
 properties {
  bUsedWithMeshDeformer=true
  bUsedWithNiagaraMeshParticles=true
  bUsedWithSkeletalMesh=true
  pixelDepthOffsetMode=PDOM_Legacy
  previewMesh=&/SM_FX_Swatch_Heart
  shadingModel=MSM_Unlit
  usedShadingModels=Unlit
 }
 nodes [
  n1 = PreSkinnedPosition()
  Add(A=@n1, B=Constant3Vector() {constant=<rgba 0.1 0 -23 0>})
  Constant3Vector() {constant=<rgba 1 0 0.0787 1>}
  n2 = VertexInterpolator(VS=Divide(A=@n1, B=Constant3Vector() {constant=<rgba 19.5 1 23.05 0>}))
  NamedRerouteDeclaration(None=MaterialFunctionCall(Vector or UVs=ComponentMask(None=@n2.PS) {b=true, r=true}) {materialFunction=&VectorToRadialValue}.Radial Coordinates) {name=uv, nodeColor=<rgba 1 0 0.8438 1>}
  n3 = NamedRerouteUsage()
  n4 = Time() {period=1}
  n5 = Frac(None=Multiply(A=@n4) {constB=-0.2})
  n6 = Frac(None=Multiply(A=@n4) {constB=1.2})
  NamedRerouteDeclaration(None=Add(A=Multiply(A=@n3) {constB=0.1}, B=AppendVector(A=@n5, B=@n6))) {name=uv move, nodeColor=<rgba 0 0.625 1 1>}
  n7 = Time() {period=1}
  n8 = Frac(None=Multiply(A=@n7) {constB=-0.1})
  n9 = Frac(None=Multiply(A=@n7) {constB=-0.8})
  NamedRerouteDeclaration(None=Add(A=NamedRerouteUsage(), B=AppendVector(A=@n8, B=@n9))) {name=uv move2, nodeColor=<rgba 0 0.625 1 1>}
  n10 = ComponentMask(None=@n2.PS) {b=true, r=true}
  NamedRerouteDeclaration(None=Distance(A=Multiply(A=@n10) {constB=1.5}, B=Constant())) {name=mask, nodeColor=<rgba 0 0.6484 1 1>}
  n11 = TextureCoordinate()
  n12 = TextureSample!(UVs=NamedRerouteUsage(), Tex, "DDX(UVs)"=DDX(Value=@n11), "DDY(UVs)"=DDY(Value=@n11), Apply View MipBias) {
   mipValueMode=TMVM_Derivative
   samplerType=SAMPLERTYPE_LinearGrayscale
   texture=/Game/Animation/Props/Gadgets_Swatch/Textures/T_Emote_Prop_Swatch_M
  }
  NamedRerouteDeclaration(None=@n12.R) {name=UVFx, nodeColor=<rgba 1 0.0703 0 1>}
  n13 = Multiply(A=Fresnel() {baseReflectFraction=0.01, exponent=3}) {constB=3}
  n14 = ParticleColor()
  n15 = NamedRerouteUsage()
  n16 = Add(A=TextureCoordinate(), B=Constant2Vector() {g=0.03})
  n17 = Add(
   A=ComponentMask(None=@n15) {g=true}
   B=Multiply(A=TextureSample!(UVs=@n16, Tex, Apply View MipBias) {samplerType=SAMPLERTYPE_LinearColor, texture=/Game/Animation/Props/Gadgets_Swatch/Textures/T_Emote_Prop_Swatch_M2}.R) {constB=-0.8}
  )
  n18 = Multiply(A=NamedRerouteUsage()) {constB=0.5}
  n19 = MaterialFunctionCall(Float2=AppendVector(A=ComponentMask(None=@n15) {r=true}, B=Add(A=@n17, B=@n18))) {materialFunction=&BreakOutFloat2Components}
  n20 = LinearInterpolate(
   A=ScalarParameter() {defaultValue=1}
   B=ScalarParameter() {defaultValue=2, parameterName=Param_1}
   Alpha=ConstantBiasScale(None=Sine(None=@n19.G))
  ) {constA=0.1, constB=1.2}
  Reroute(None=Add(A=Multiply(A=@n13, B=@n14.RGB), B=Multiply(A=@n14.RGB, B=@n20)))
  Reroute(None=@n12.R)
 ]
}
